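# Additive, unlit, two-sided glow material for Niagara sprite and mesh particles, in the BikeMold bass cosmetic's effects.
# Particle color feeds an emissive-adjustment function, alongside a time-of-day multiplier set equally for all four periods.
# The alpha branch blends a procedural exponential radial gradient with a camera-flare texture, then multiplies by particle alpha and a view-angle fade.
# A separate radial generated band feeds a SolidorHollow blend.
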Material /SparksCosmetics/Instrument/Bass/BikeMold/Effects/M_Bass_BikeMold_Glow
graph {
 properties {
  bAutomaticallySetUsageInEditor=false
  bUsedWithNiagaraMeshParticles=true
  bUsedWithNiagaraSprites=true
  blendMode=BLEND_Additive
  pixelDepthOffsetMode=PDOM_Legacy
  shadingModel=MSM_Unlit
  twoSided=true
  usedShadingModels=Unlit
 }
 nodes [
  Constant() {r=1}
  n1 = Constant() {r=0.05}
  n2 = Constant() {r=-5}
  n3 = Constant() {r=-5}
  n4 = MaterialFunctionCall() {materialFunction=&VectorToRadialValue}
  LinearInterpolate(
   B=Multiply(A=MaterialFunctionCall!(Width=@n1, Sharpness=@n2, Offset=@n3, Direction Switch, Compare, Input Coordinates=@n4.Radial Coordinates) {materialFunction=/Engine/Functions/Engine_MaterialFunctions03/Procedurals/GeneratedBand}.Result) {constB=3}
   Alpha=ScalarParameter() {parameterName=SolidorHollow}
  )
  n5 = ParticleColor()
  MaterialFunctionCall(Emissive=@n5.RGB) {materialFunction=&MF_FX_EmissiveAdjustments}
  n6 = Constant() {r=1}
  Multiply(A=MaterialFunctionCall(Color_Morning=@n6, Color_Day=Constant() {r=1}, Color_Evening=Constant() {r=1}, Color_Night=Constant() {r=1}) {materialFunction=&Color_Time_of_Day}.Result)
  n7 = MaterialFunctionCall(Radius=ScalarParameter() {defaultValue=0.5, parameterName=Radius}, Density=ScalarParameter() {defaultValue=2, parameterName=Desity}) {materialFunction=&RadialGradientExponential}
  n8 = TextureSampleParameter2D() {parameterName=Texture, samplerType=SAMPLERTYPE_Grayscale, texture=/Game/Effects/Fort_Effects/Textures/T_EFX_CameraFlare_01}
  n9 = ScalarParameter() {parameterName=Mat_or_Tex_Alpha}
  Multiply(
   A=Multiply(A=@n5.A, B=LinearInterpolate(A=@n7.RadialGradientExponential, B=@n8.G, Alpha=@n9))
   B=MaterialFunctionCall("FadeSensitivity (optional)"=@n6) {materialFunction=&FadeBasedOnViewAngle}.Result
  )
 ]
}
