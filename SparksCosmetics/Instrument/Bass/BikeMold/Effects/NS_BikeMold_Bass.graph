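# Niagara glow effect for the BikeMold bass instrument cosmetic, with a blue circle glow and a larger lens-style flare.
# Both emitters burst-spawn particles with looping lifetimes, so they persist as offset sprites with fixed facing, pulled by camera offset.
# CircleGlow also drives a light renderer for blue illumination.
# Flare's local-space alignment switches on User.bIsFrontEndPreview, changing its tilt for front-end preview.
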NiagaraSystem /SparksCosmetics/Instrument/Bass/BikeMold/Effects/NS_BikeMold_Bass
dependencies {usedDataInterfaces [], usedDynamicInputs [], usedModules [&SystemState]}
emitters {
 CircleGlow {
  emitterUpdateScript [
   &EmitterState(Life Cycle Mode=::NewEnumerator1, Loop Behavior=::NewEnumerator1, Loop Duration Mode=::NewEnumerator1, UseLoopDelay=false)
   &SpawnBurst_Instantaneous(Use Spawn Probability=false, UseLoopCountLimit=false)
  ]
  particleSpawnScript [
   &InitializeParticle(
    Color=<rgba 0 0.5 1 0.2>
    Lifetime=2
    Position Offset=<0 1.5 -5>
    Sprite Size Mode=::NewEnumerator3
    Uniform Sprite Size=19
    UsePositionOffset=true
    Write Material Random=true
   )
   /SetVariables(Particles.LightSpecularScale=50) {isSetParameters=true}
   /SetVariables(Particles.Light Diffuse=10) {isSetParameters=true}
  ]
  particleUpdateScript [
   &ParticleState(
    Kill Particles When Lifetime Has Elapsed=false
    Let Infinitely Lived Particles Die When Emitter Deactivates=true
    Loop Particles Lifetime=true
   )
   &SpriteFacingAndAlignment(Sprite Alignment=<-1 0 0>, Sprite Facing=<0 1 0>, WriteAlignment=true, WriteFacing=true)
   &SolveForcesAndVelocity(Clamp Velocity=false, Limit Acceleration=false)
   &FN_Light_Attributes(
    Color=<rgba 0 0.5 1 1>
    Create Color Attribute=true
    Create Exponent Attribute=true
    Create Radius Attribute=true
    Exponent=0.2
    Radius=10
    Scale Color=0.2
   )
   &CameraOffset(Camera Offset Amount=1)
  ]
  rendererClasses [&NiagaraSpriteRendererProperties, &NiagaraLightRendererProperties]
  renderers [
   {rendererClass=&NiagaraSpriteRendererProperties}
   {rendererClass=&NiagaraLightRendererProperties, rendererIndex=1}
  ]
 }
 Flare {
  emitterUpdateScript [
   &EmitterState(Life Cycle Mode=::NewEnumerator1, Loop Behavior=::NewEnumerator1, Loop Duration Mode=::NewEnumerator1, UseLoopDelay=false)
   &SpawnBurst_Instantaneous(Use Spawn Probability=false, UseLoopCountLimit=false)
  ]
  particleSpawnScript [&InitializeParticle(
    Color=<rgba 0.3588 0.498 1 0.5>
    Lifetime=2
    Position Offset=<0 2 -5>
    Sprite Rotation Angle=90
    Sprite Rotation Mode=::NewEnumerator0
    Sprite Size Mode=::NewEnumerator3
    Uniform Sprite Size=80
    UsePositionOffset=true
    Write Material Random=true
   )]
  particleUpdateScript [
   &ParticleState(
    Kill Particles When Lifetime Has Elapsed=false
    Let Infinitely Lived Particles Die When Emitter Deactivates=true
    Loop Particles Lifetime=true
   )
   &SpriteFacingAndAlignment(
    Alignment Coordinate Space=::Local
    Sprite Alignment=&MakeCustomVectorFromBool(Bool=$User.bIsFrontEndPreview, True Vector=<1 0 0.45>, False Vector=<1 0 1.6>)
    Sprite Facing=<0 1 0>
    WriteAlignment=true
    WriteFacing=true
   )
   &SolveForcesAndVelocity(Clamp Velocity=false, Limit Acceleration=false)
   &CameraOffset(Camera Offset Amount=1)
  ]
  rendererClasses [&NiagaraSpriteRendererProperties]
  renderers [{rendererClass=&NiagaraSpriteRendererProperties}]
 }
}
summary {
 systemName=NS_BikeMold_Bass
 emitters [
  {emitterName=CircleGlow, rendererClasses [&NiagaraSpriteRendererProperties, &NiagaraLightRendererProperties]}
  {emitterName=Flare, rendererClasses [&NiagaraSpriteRendererProperties]}
 ]
 userVariables [User.bIsFrontEndPreview {type=NiagaraBool}]
}
